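# Material in the engine's map template Sky folder that samples the DaylightAmbientCubemap texture.
# A single texture sample uses the world-space reflection vector as its lookup coordinate, the usual way to map a cubemap by reflection.
# It uses Default Lit shading and is flagged for use with static lighting, and the graph shows no connection from the sample to a material output.
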Material /Engine/MapTemplates/Sky/DaylightAmbientCubemap_Mat
graph {
 properties {bUsedWithStaticLighting=true, pixelDepthOffsetMode=PDOM_Legacy, usedShadingModels=Default Lit}
 nodes [TextureSample(UVs=ReflectionVectorWS()) {texture=&/DaylightAmbientCubemap}]
}
