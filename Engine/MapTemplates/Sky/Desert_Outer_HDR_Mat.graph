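# Default Lit material in the engine's map-template Sky folder, built around the Desert_Outer_HDR environment texture.
# A single texture sample reads that HDR texture in linear color, using the world-space reflection vector as its coordinates.
# The graph shows no connection from the sample to any material input, so the texture does not visibly drive the shading.
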Material /Engine/MapTemplates/Sky/Desert_Outer_HDR_Mat
graph {
 properties {pixelDepthOffsetMode=PDOM_Legacy, usedShadingModels=Default Lit}
 nodes [TextureSample(UVs=ReflectionVectorWS()) {samplerType=SAMPLERTYPE_LinearColor, texture=&/Desert_Outer_HDR}]
}
