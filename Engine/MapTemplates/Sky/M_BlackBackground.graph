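# M_BlackBackground is an unlit material in the engine's map template sky folder, named as a black background.
# The graph holds a single Constant node at its default value of zero, with no textures, parameters or math.
# It is flagged for use with static lighting.
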Material /Engine/MapTemplates/Sky/M_BlackBackground
graph {
 properties {bUsedWithStaticLighting=true, pixelDepthOffsetMode=PDOM_Legacy, shadingModel=MSM_Unlit, usedShadingModels=Unlit}
 nodes [Constant()]
}
