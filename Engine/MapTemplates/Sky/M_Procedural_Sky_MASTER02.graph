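# Procedural sky material, apparently ported from UE3, with a zenith-to-horizon gradient, optional twinkling stars, animated clouds and a sun.
# Panning cloud masks control cloud opacity and hide the sun, while blended normals against the light vector drive cloud shading and rim glow.
# A sphere mask on camera-to-light alignment draws the sun disc, sized by a static switch.
# Gradient and cloud-lighting branches end separately, with no material inputs connected.
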Material /Engine/MapTemplates/Sky/M_Procedural_Sky_MASTER02
graph {
 properties {bUsedWithStaticLighting=true, pixelDepthOffsetMode=PDOM_Legacy, usedShadingModels=Default Lit}
 nodes [
  n1 = Multiply(A=TextureCoordinate() {coordinateIndex=2}, B=ScalarParameter() {defaultValue=5, parameterName=StarTiling})
  n2 = TextureSample(UVs=Panner(Coordinate=@n1) {speedX=0.02, speedY=0.02}) {texture=&/T_Sky_Stars}
  n3 = Constant() {r=5}
  n4 = TextureSample(UVs=@n1) {texture=&T_Sky_Stars}
  n5 = Add(A=Multiply(A=Multiply(A=@n2.RGB, B=@n3), B=@n4.RGB), B=@n4.RGB)
  n6 = VectorParameter() {parameterName=StarColor, defaultValue=<rgba 1 1 2 4>}
  n7 = Multiply(A=@n6.RGB, B=@n6.RGB)
  n8 = ComponentMask(None=TextureCoordinate() {desc=CYL}) {g=true}
  n9 = OneMinus(None=Multiply(A=@n8, B=@n8))
  n10 = Multiply(A=@n9, B=@n9)
  n11 = Desaturation(
   None=LinearInterpolate(
    A=VectorParameter() {parameterName=ZenithColor, defaultValue=<rgba 0.126 0.246 0.684 1>}.RGB
    B=VectorParameter() {parameterName=HorizonColor, defaultValue=<rgba 0.238 0.39 0.47 1>}.RGB
    Alpha=OneMinus(None=@n10)
   )
   Fraction=ScalarParameter() {parameterName=Desaturation}
  )
  n12 = StaticSwitchParameter(True=Add(A=Multiply(A=@n5, B=@n7), B=@n11), False=@n11) {parameterName=Stars}
  n13 = ScalarParameter() {defaultValue=0.5, parameterName=Speed}
  n14 = Multiply(A=Time(), B=Multiply(A=@n13, B=Constant() {r=4}))
  n15 = Panner(Time=@n14) {speedX=0.001}
  n16 = TextureSample(UVs=@n15) {samplerType=SAMPLERTYPE_Grayscale, texture=/Engine/MapTemplates/T_UDK_Sky_CloudMask03}
  n17 = Panner(Coordinate=TextureCoordinate() {uTiling=-1}, Time=@n14) {speedX=-0.002}
  n18 = Add(
   A=@n16.R
   B=Multiply(A=Constant() {r=0.5}, B=TextureSample(UVs=@n17) {samplerType=SAMPLERTYPE_Grayscale, texture=&T_UDK_Sky_CloudMask03}.R)
  )
  n19 = Multiply(A=@n18, B=Constant() {r=0.5})
  n20 = Time()
  n21 = TextureSample(UVs=Panner(Coordinate=TextureCoordinate() {uTiling=-1}, Time=Multiply(A=@n13, B=@n20)) {speedX=-0.02}) {desc=UE3 this was T_UDK_Sky_CloudMask01, samplerType=SAMPLERTYPE_Grayscale, texture=&T_UDK_Sky_CloudMask03}
  n22 = Multiply(A=Clamp(None=LinearInterpolate(A=@n19, B=@n18, Alpha=@n21.R)), B=ScalarParameter() {defaultValue=1, parameterName=CloudOpacity})
  LinearInterpolate(A=@n12, B=Constant(), Alpha=@n22) {desc=In UE3 this was connected to the Multiply A pin}
  n23 = ScalarParameter() {defaultValue=1, parameterName=SkyBrightness}
  Multiply(A=@n12, B=@n23)
  n24 = TextureSample(UVs=@n17) {samplerType=SAMPLERTYPE_LinearColor, texture=&/T_UDK_Sky_CloudMask01_N}
  n25 = TextureSample(UVs=@n15) {samplerType=SAMPLERTYPE_LinearColor, texture=&T_UDK_Sky_CloudMask01_N}
  n26 = LinearInterpolate(A=@n24.RGB, B=@n25.RGB, Alpha=@n16.R)
  n27 = Constant()
  n28 = ConstantBiasScale(None=DotProduct(A=Add(A=@n26, B=@n27), B=LightVector()))
  n29 = Multiply(
   A=@n22
   B=LinearInterpolate(
    A=ScalarParameter() {defaultValue=0.05, parameterName=CloudDarkness}
    B=ScalarParameter() {defaultValue=1, parameterName=CloudBrightness}
    Alpha=Multiply(A=@n28, B=@n28)
   )
  )
  n30 = Add(A=@n25.RGB, B=Multiply(A=@n24.RGB, B=Constant3Vector() {constant=<rgba 1 1 1 0>}))
  n31 = Multiply(A=Fresnel(Normal=Normalize(VectorInput=@n30)) {exponent=1}, B=ScalarParameter() {defaultValue=2, parameterName=RimBrightness})
  n32 = CameraVectorWS()
  n33 = LightVector()
  n34 = Clamp(None=Multiply(A=DotProduct(A=@n32, B=@n33), B=Constant() {r=-1}))
  n35 = Power(Base=@n34, Exp=Constant() {r=10})
  n36 = VectorParameter() {parameterName=RimColor, defaultValue=<rgba 10 6 2 1>}
  n37 = Multiply(A=@n22, B=Multiply(A=Multiply(A=@n31, B=@n35), B=Multiply(A=@n36.RGB, B=@n36.RGB)))
  n38 = OneMinus(None=@n22)
  n39 = Constant() {r=1}
  n40 = VectorParameter() {parameterName=Sun, defaultValue=<rgba 10 6 2 1>}
  n41 = Multiply(
   A=StaticSwitchParameter(
    True=SphereMask(A=@n34, B=@n39) {attenuationRadius=0.002, hardnessPercent=0}
    False=SphereMask(A=@n34, B=@n39) {attenuationRadius=0.001, hardnessPercent=0}
   ) {parameterName=LargeSun}
   B=Multiply(A=@n40.RGB, B=@n40.RGB)
  )
  Multiply(A=@n23, B=Add(A=Add(A=@n29, B=@n37), B=Multiply(A=@n38, B=@n41))) {desc=In UE3 this was connected to CustomLighting Pin}
 ]
}
