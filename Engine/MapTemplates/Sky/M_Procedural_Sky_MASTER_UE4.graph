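# Unlit master sky material that displays a cubemap as the sky, despite the "Procedural" in its name.
# It negates the world-space camera vector to get the view direction and uses that to sample the CubeMap parameter, which defaults to DaylightAmbientCubemap.
# The sampled RGB is multiplied by the SkyBrightness scalar, and both parameters sit in the Sky group for tuning.
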Material /Engine/MapTemplates/Sky/M_Procedural_Sky_MASTER_UE4
graph {
 properties {bUsedWithStaticLighting=true, pixelDepthOffsetMode=PDOM_Legacy, shadingModel=MSM_Unlit, usedShadingModels=Unlit}
 nodes [
  n1 = CameraVectorWS()
  n2 = Constant() {r=-1}
  Multiply(
   A=TextureSampleParameterCube!(UVs=Multiply(A=@n1, B=@n2), Apply View MipBias) {group=Sky, parameterName=CubeMap, texture=&/DaylightAmbientCubemap}.RGB
   B=ScalarParameter() {defaultValue=1, group=Sky, parameterName=SkyBrightness}
  )
 ]
}
